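# Translucent unlit material for the player marker, building a glow and reshaping its mesh by distance from the camera.
# Horizontal distance between ObjectPivotLoc and the camera drives staged blends of opacity, brightness and size.
# Axis-aligned fresnels, UV falloffs and vertex color shape the glow, with depth fade except on mobile.
# Offset math raises the blue-masked tip and scales along normals, with spawn parameters modulating opacity and scale.
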Material /Game/Athena/HUD/Art/PlayerMarker/M_PlayerMarker
graph {
 properties {
  bTangentSpaceNormal=false
  bUseTranslucencyVertexFog=false
  bUsedWithNanite=true
  blendMode=BLEND_Translucent
  ditherOpacityMask=true
  opacityMaskClipValue=0.33
  pixelDepthOffsetMode=PDOM_Legacy
  shadingModel=MSM_Unlit
  usedShadingModels=Unlit
 }
 nodes [
  n1 = VertexColor()
  n2 = OneMinus(None=@n1.B)
  n3 = Distance(
   A=ComponentMask(None=VectorParameter() {parameterName=ObjectPivotLoc, defaultValue=<rgba 1 1 1 1>}.RGB) {g=true, r=true}
   B=ComponentMask(None=CameraPositionWS()) {g=true, r=true}
  )
  n4 = ScalarParameter() {defaultValue=40000, parameterName=S0-DistanceOffset}
  n5 = Clamp(None=Multiply(A=Subtract(A=@n3, B=@n4), B=ScalarParameter() {defaultValue=0, parameterName=S0-DistanceScale}))
  n6 = MaterialFunctionCall(Float2=LinearInterpolate(A=Constant2Vector() {g=10, r=4.75}, B=Constant2Vector() {g=1.5, r=1}, Alpha=@n5)) {materialFunction=&BreakOutFloat2Components}
  n7 = ScalarParameter() {defaultValue=4, parameterName=InnerFresnelExponent}
  n8 = Clamp(None=ComponentMask(None=TextureCoordinate() {coordinateIndex=2}) {g=true}) {maxDefault=0.9999, minDefault=0.0001}
  n9 = OneMinus(None=@n8)
  n10 = Power(Base=@n9, Exp=@n6.R)
  n11 = ScalarParameter() {defaultValue=2, parameterName=InnerFresVertFalloffMod}
  n12 = Multiply(
   A=Multiply(
    A=@n6.G
    B=MaterialFunctionCall(Exponent=@n7) {materialFunction=&AxisAlignedFresnel}.Fresnel
   )
   B=Clamp(None=Multiply(A=@n10, B=@n11))
  )
  n13 = Multiply(
   A=Power(Base=@n8, Exp=ScalarParameter() {defaultValue=3, parameterName=OL-VerticalFalloffPow})
   B=ScalarParameter() {defaultValue=5, parameterName=OL-VerticalFalloffMod}
  )
  n14 = ScalarParameter() {defaultValue=0.25, parameterName=OL-ExpoClose}
  n15 = ScalarParameter() {defaultValue=0.1, parameterName=OL-ExpoFar}
  n16 = Subtract(A=@n3, B=ScalarParameter() {defaultValue=11700, parameterName=S2-DistanceOffset})
  n17 = Clamp(None=Multiply(A=@n16, B=ScalarParameter() {defaultValue=0, parameterName=S0-DistanceScale_1}))
  n18 = OneMinus(None=MaterialFunctionCall(Exponent=LinearInterpolate(A=@n14, B=@n15, Alpha=@n17)) {materialFunction=&AxisAlignedFresnel}.Fresnel)
  n19 = Multiply(A=ScalarParameter() {defaultValue=25, parameterName=OL-InnerMod}, B=@n18)
  n20 = ScalarParameter() {defaultValue=28000, parameterName=S1-DistanceOffset}
  n21 = Clamp(None=Multiply(A=Subtract(A=@n3, B=@n20), B=ScalarParameter() {defaultValue=0.0001, parameterName=S1-DistanceScale}))
  n22 = LinearInterpolate(A=Constant2Vector() {g=0.8, r=2}, B=Constant2Vector() {g=1, r=1}, Alpha=@n21)
  n23 = ComponentMask(None=@n22) {r=true}
  n24 = Multiply(
   A=Multiply(
    A=Clamp(None=@n19)
    B=MaterialFunctionCall(Exponent=@n23) {materialFunction=&AxisAlignedFresnel}.Fresnel
   )
   B=ComponentMask(None=@n22) {g=true}
  )
  n25 = CameraPositionWS()
  n26 = Multiply(
   A=Subtract(A=ComponentMask(None=@n25) {b=true}, B=ScalarParameter() {defaultValue=25000, parameterName=HeightOffset})
   B=ScalarParameter() {defaultValue=0, parameterName=HeightScale}
  )
  n27 = Clamp(None=ComponentMask(None=TextureCoordinate() {coordinateIndex=1}) {g=true})
  n28 = Multiply(A=Multiply(A=Clamp(None=@n13), B=@n24), B=LinearInterpolate(A=Constant() {r=1}, B=@n27, Alpha=Clamp(None=@n26)))
  n29 = Multiply(A=Multiply(A=Multiply(A=@n1.R, B=@n2), B=Add(A=@n12, B=@n28)), B=ScalarParameter() {defaultValue=1, parameterName=SpawnOpacity})
  n30 = ScalarParameter() {defaultValue=5000, parameterName=S4-DistanceOffset}
  n31 = Clamp(None=Multiply(A=Subtract(A=@n3, B=@n30), B=ScalarParameter() {defaultValue=0.001, parameterName=S4-DistanceScale}))
  n32 = ScalarParameter() {defaultValue=3000, parameterName=S3-DistanceOffset}
  n33 = Clamp(None=Multiply(A=Subtract(A=@n3, B=@n32), B=ScalarParameter() {defaultValue=0.0001, parameterName=S3-DistanceScale}))
  n34 = Multiply(A=@n31, B=@n33)
  n35 = Multiply(A=@n29, B=@n34)
  n36 = Multiply(A=ScalarParameter() {parameterName=SpawnScale}, B=@n27)
  n37 = Multiply(A=Multiply(A=Constant() {r=1}, B=@n21), B=@n36)
  Add(A=@n35, B=Multiply(A=@n35, B=Clamp(None=@n37)))
  n38 = TextureCoordinate()
  n39 = MaterialFunctionCall(X=Constant() {r=-0.4}, Y=Constant() {r=0.3}) {materialFunction=&MakeFloat2}
  n40 = Multiply(A=@n33, B=Constant())
  n41 = Clamp(None=Multiply(A=@n16, B=ScalarParameter() {defaultValue=0.0001, parameterName=S2-DistanceScale}))
  n42 = Multiply(A=@n41, B=Constant())
  n43 = Constant()
  n44 = Add(A=Add(A=Add(A=@n40, B=@n42), B=Multiply(A=@n21, B=@n43)), B=Multiply(A=@n5, B=Constant() {r=1}))
  n45 = Multiply(A=@n17, B=Constant())
  n46 = TextureCoordinate()
  n47 = MaterialFunctionCall(X=Constant(), Y=Constant() {r=-0.2}) {materialFunction=&MakeFloat2}
  Add(
   A=MaterialFunctionCall(Number=@n3, MaximumNumberOfDigits=Constant() {r=6}, UVs=Add(A=@n38, B=@n39.Result)) {materialFunction=&DebugScalarValues}.Result
   B=MaterialFunctionCall(Number=Add(A=@n44, B=@n45), MaximumNumberOfDigits=Constant() {r=6}, UVs=Add(A=@n46, B=@n47.Result)) {materialFunction=&DebugScalarValues}.Result
  )
  Clamp(None=@n9)
  Constant3Vector() {constant=<rgba 0 0.0971 0.4 1>}
  Constant3Vector() {constant=<rgba 0 0.3641 1.5 1>}
  Constant()
  Constant() {r=0.25}
  Constant() {r=0.1}
  Constant() {r=5}
  Constant() {r=3}
  Constant() {r=0}
  Constant() {r=0}
  Constant() {r=2000}
  Constant() {r=40000}
  Constant() {r=20000}
  Constant() {r=0}
  Constant() {r=0}
  Constant() {r=30000}
  Constant() {r=1500}
  Constant() {r=3000}
  Constant() {r=-350}
  Constant() {r=350}
  Constant() {r=2000}
  Constant() {r=10000}
  Constant() {r=20000}
  Constant() {r=-2000}
  Constant() {r=-6000}
  Constant() {r=40000}
  n48 = ScalarParameter() {defaultValue=200, parameterName=SpawnScaleModClose}
  n49 = ScalarParameter() {defaultValue=400, parameterName=SpawnScaleModFar}
  LinearInterpolate(A=@n48, B=@n49, Alpha=@n17)
  n50 = ScalarParameter() {parameterName=BottomGirthFar}
  n51 = ScalarParameter() {parameterName=BottomGirthFarthest}
  LinearInterpolate(A=ScalarParameter() {parameterName=BottomGirthClose}, B=LinearInterpolate(A=@n50, B=@n51, Alpha=@n5), Alpha=@n21)
  n52 = Multiply(A=@n29, B=@n31)
  n53 = Multiply(A=@n52, B=ScalarParameter() {defaultValue=0.25, parameterName=FarDarkTop})
  LinearInterpolate(A=@n52, B=@n53, Alpha=@n5)
  n54 = Constant()
  n55 = ScalarParameter() {defaultValue=1000, parameterName=DepthFadeClose}
  n56 = ScalarParameter() {defaultValue=2000, parameterName=DepthFadeFar}
  n57 = MaterialFunctionCall(Float2=LinearInterpolate(
    A=MaterialFunctionCall(X=@n52, Y=@n55) {materialFunction=&MakeFloat2}.Result
    B=MaterialFunctionCall(X=@n53, Y=@n56) {materialFunction=&MakeFloat2}.Result
    Alpha=@n5
   )) {materialFunction=&BreakOutFloat2Components}
  n58 = Multiply(A=CameraVectorWS(), B=Constant3Vector() {constant=<rgba 1 1 4 1>})
  n59 = Constant3Vector() {constant=<rgba 0 0 -3 1>}
  n60 = DotProduct(A=Constant3Vector() {constant=<rgba 0 0 1 1>}, B=Normalize(VectorInput=Add(A=@n58, B=@n59)))
  n61 = Multiply(A=@n35, B=OneMinus(None=Clamp(None=@n60)))
  LinearInterpolate(
   A=ShadingPathSwitch(Default=DepthFade(Opacity=@n61, FadeDistance=@n57.G) {fadeDistanceDefault=30}, Mobile=@n61) {desc=Migrated from FeatureLevelSwitch}
   Alpha=@n54
  )
  n62 = VectorParameter() {parameterName=MainColor, defaultValue=<rgba 0 0.2427 1 1>}
  n63 = ScalarParameter() {defaultValue=0.5, parameterName=Color-Darkest}
  n64 = ScalarParameter() {defaultValue=1.25, parameterName=Color-Brightest}
  LinearInterpolate(A=LinearInterpolate(A=Multiply(A=@n62.RGB, B=@n63), B=Multiply(A=@n62.RGB, B=@n64), Alpha=@n57.R), B=@n24, Alpha=@n54)
  n65 = ScalarParameter() {defaultValue=-350, parameterName=OverallSizeClose}
  n66 = ScalarParameter() {defaultValue=350, parameterName=OverallSizeFar}
  LinearInterpolate(A=@n65, B=@n66, Alpha=@n17)
  n67 = Constant()
  n68 = ScalarParameter() {defaultValue=-500, parameterName=MeshHeightFar}
  LinearInterpolate(A=@n67, B=@n68, Alpha=@n17)
  LinearInterpolate(A=@n55, B=@n56, Alpha=@n5)
  n69 = ScalarParameter() {parameterName=BallUpperOffsetFar}
  n70 = ScalarParameter() {parameterName=BallUpperOffsetFarthest}
  LinearInterpolate(A=@n69, B=@n70, Alpha=@n5)
  Power(Base=@n18, Exp=Constant() {r=1.5})
  Reroute(None=@n41)
  n71 = MaterialFunctionCall(Float2=LinearInterpolate(
    A=MaterialFunctionCall(X=@n69, Y=@n50) {materialFunction=&MakeFloat2}.Result
    B=MaterialFunctionCall(X=@n70, Y=@n51) {materialFunction=&MakeFloat2}.Result
    Alpha=@n5
   )) {materialFunction=&BreakOutFloat2Components}
  n72 = LinearInterpolate(A=ScalarParameter() {parameterName=BallUpperOffsetClose}, B=@n71.R, Alpha=@n21)
  n73 = Constant3Vector() {constant=<rgba 0 0 1 0>}
  Reroute(None=Multiply(A=Reroute(), B=Multiply(A=@n72, B=@n73)))
  n74 = TextureCoordinate()
  n75 = Power(Base=Clamp(None=ComponentMask(None=@n74) {g=true}), Exp=ScalarParameter() {defaultValue=2, parameterName=SphereTopCurve})
  Reroute(None=Multiply(A=@n75, B=@n71.G))
  Reroute(None=@n21)
  n76 = ScalarParameter() {defaultValue=5000, parameterName=TipHeightClose}
  n77 = ScalarParameter() {defaultValue=1, parameterName=SpawnScaleVertical}
  n78 = LinearInterpolate(
   A=ScalarParameter() {defaultValue=20000, parameterName=TipHeightFar}
   B=ScalarParameter() {defaultValue=40000, parameterName=TipHeightFarthest}
   Alpha=Multiply(A=@n17, B=@n77)
  )
  n79 = Multiply(A=@n77, B=@n41)
  n80 = Multiply(A=Multiply(A=Constant3Vector() {constant=<rgba 0 0 1 0>}, B=LinearInterpolate(A=@n76, B=@n78, Alpha=@n79)), B=VertexColor().B)
  n81 = MaterialFunctionCall(Float3=LinearInterpolate(
    A=MaterialFunctionCall(X=@n65, Y=@n48, Z=@n67) {materialFunction=&MakeFloat3}.Result
    B=MaterialFunctionCall(X=@n66, Y=@n49, Z=@n68) {materialFunction=&MakeFloat3}.Result
    Alpha=@n17
   )) {materialFunction=&BreakOutFloat3Components}
  n82 = Add(A=Multiply(A=Constant3Vector() {constant=<rgba 0 0 1 0>}, B=@n81.B), B=@n27)
  n83 = VertexNormalWS()
  n84 = Multiply(A=Constant() {r=-595}, B=@n83)
  n85 = Multiply(A=@n36, B=@n81.G)
  n86 = OneMinus(None=@n75)
  n87 = Multiply(A=@n75, B=ScalarParameter() {defaultValue=-600, parameterName=BottomCreatesPoint})
  n88 = Constant()
  n89 = LinearInterpolate(A=Add(A=Multiply(A=@n86, B=@n81.R), B=LinearInterpolate(A=@n87, B=@n88, Alpha=@n41)), B=@n81.R, Alpha=@n41)
  n90 = Multiply(A=@n83, B=Clamp(None=Add(A=@n85, B=@n89), Min=ScalarParameter() {defaultValue=-600, parameterName=ClampScaleMin}) {clampMode=CMODE_ClampMin})
  n91 = Clamp(None=Multiply(A=Constant() {r=3}, B=@n34))
  Reroute(None=Add(A=Add(A=@n80, B=@n82), B=LinearInterpolate(A=@n84, B=@n90, Alpha=@n91)))
 ]
}
