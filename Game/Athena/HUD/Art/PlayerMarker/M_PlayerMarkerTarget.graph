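# Unlit translucent, two-sided material for the player marker target in the HUD art, drawing scrolling, tinted bands masked by vertex red.
# Camera distance to ObjectPivotLoc and camera height form a mask that drives visibility and scaling around that pivot, plus a Z offset.
# A second distance blend shifts size and band sharpness between near and far settings.
# SpawnOpacity and SpawnScale modulate opacity and color, with depth fade skipped on mobile.
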Material /Game/Athena/HUD/Art/PlayerMarker/M_PlayerMarkerTarget
graph {
 properties {
  bTangentSpaceNormal=false
  bUseTranslucencyVertexFog=false
  bUsedWithNanite=true
  blendMode=BLEND_Translucent
  ditherOpacityMask=true
  opacityMaskClipValue=0.33
  pixelDepthOffsetMode=PDOM_Legacy
  shadingModel=MSM_Unlit
  twoSided=true
  usedShadingModels=Unlit
 }
 nodes [
  n1 = WorldPosition()
  n2 = VectorParameter() {parameterName=ObjectPivotLoc, defaultValue=<rgba 1 1 1 1>}
  n3 = ScalarParameter() {defaultValue=-1, parameterName=Size-Zero}
  n4 = Distance(A=VectorParameter() {parameterName=ObjectPivotLoc, defaultValue=<rgba 1 1 1 1>}.RGB, B=CameraPositionWS())
  n5 = ScalarParameter() {defaultValue=40000, parameterName=BlendDistance-Offset}
  n6 = Clamp(None=Multiply(A=Subtract(A=@n4, B=@n5), B=ScalarParameter() {defaultValue=0, parameterName=BlendDistance-Scale}))
  n7 = LinearInterpolate(A=ScalarParameter() {defaultValue=2, parameterName=Size-Close}, B=ScalarParameter() {defaultValue=6, parameterName=Size-Far}, Alpha=@n6)
  n8 = Subtract(A=@n4, B=ScalarParameter() {defaultValue=15000, parameterName=MasterDistance-Offset})
  n9 = ScalarParameter() {defaultValue=0, parameterName=MasterDistance-Scale}
  n10 = Subtract(A=ComponentMask(None=CameraPositionWS()) {b=true}, B=ScalarParameter() {defaultValue=25000, parameterName=HeightOffset})
  n11 = ScalarParameter() {defaultValue=0, parameterName=HeightScale}
  n12 = Multiply(A=Clamp(None=Multiply(A=@n8, B=@n9)), B=Clamp(None=Multiply(A=@n10, B=@n11)))
  n13 = Clamp(None=Multiply(A=@n12, B=ScalarParameter() {defaultValue=8, parameterName=ScaleMaskMod}))
  Add(
   A=Multiply(A=Constant3Vector() {constant=<rgba 0 0 1 0>}, B=ScalarParameter() {defaultValue=250, parameterName=ZOffset})
   B=Multiply(A=Subtract(A=@n1.XYZ, B=@n2.RGB), B=LinearInterpolate(A=@n3, B=@n7, Alpha=@n13))
  )
  n14 = Add(A=Multiply(A=CameraVectorWS(), B=Constant3Vector() {constant=<rgba 1 1 1 1>}), B=Constant3Vector() {constant=<rgba 0 0 -0.5 1>})
  Clamp(None=DotProduct(A=Constant3Vector() {constant=<rgba 0 0 1 1>}, B=Normalize(VectorInput=@n14)))
  n15 = Constant()
  n16 = ScalarParameter() {defaultValue=500, parameterName=DepthFade}
  n17 = ScalarParameter() {defaultValue=1, parameterName=SpawnOpacity}
  n18 = ScalarParameter() {defaultValue=1.25, parameterName=Brightness}
  n19 = Constant() {r=4}
  n20 = Multiply(A=TextureCoordinate(), B=ScalarParameter() {defaultValue=2, parameterName=Tiles})
  n21 = ScalarParameter() {defaultValue=-0.633, parameterName=Speed}
  n22 = Frac(None=ComponentMask(None=Panner(Coordinate=@n20, Speed=@n21) {speedY=-1}) {g=true})
  n23 = OneMinus(None=@n22)
  n24 = Multiply(A=Clamp(None=Multiply(A=@n19, B=@n23)), B=@n22)
  n25 = ScalarParameter() {defaultValue=2, parameterName=PowerClose}
  n26 = ScalarParameter() {defaultValue=1, parameterName=PowerFar}
  n27 = Multiply(A=VertexColor().R, B=Power(Base=Multiply(A=@n18, B=@n24), Exp=LinearInterpolate(A=@n25, B=@n26, Alpha=@n6)))
  n28 = ScalarParameter() {parameterName=SpawnScale}
  n29 = Add(A=@n27, B=Multiply(A=@n27, B=Clamp(None=@n28)))
  n30 = Multiply(A=@n17, B=Multiply(A=@n29, B=@n12))
  LinearInterpolate(
   A=ShadingPathSwitch(Default=DepthFade(Opacity=@n30, FadeDistance=@n16) {fadeDistanceDefault=30}, Mobile=@n30) {desc=Migrated from FeatureLevelSwitch}
   Alpha=@n15
  )
  n31 = VectorParameter() {parameterName=MainColor, defaultValue=<rgba 0 0.2427 1 1>}
  n32 = ScalarParameter() {defaultValue=0.5, parameterName=Color-Darkest}
  n33 = ScalarParameter() {defaultValue=1.25, parameterName=Color-Brightest}
  n34 = Power(Base=@n29, Exp=ScalarParameter() {defaultValue=2, parameterName=PowerColor})
  LinearInterpolate(A=LinearInterpolate(A=Multiply(A=@n31.RGB, B=@n32), B=Multiply(A=@n31.RGB, B=@n33), Alpha=Multiply(A=@n34, B=@n17)), Alpha=@n15)
 ]
}
